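# Niagara system for the off-hand swing trail of the WrenchPeel pickaxe skin.
# Grey, red and yellow ribbon emitters use Fortnite anim-trail modules to spawn, initialize and update short-lived trails, each between its own pair of bone or socket indices.
# ScaleColor tints each ribbon by a time-of-day color and scales alpha by a system-level SystemFade value.
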NiagaraSystem /BRCosmetics/Weapons/FORT_Melee/Pickaxe_MollyKit/Skins/WrenchPeel/FX/NS_Pickaxe_WrenchPeel_Trail_Offhand
dependencies {
 usedDataInterfaces [&NiagaraDataInterfaceSocketReader, &NiagaraDataInterfaceCurve]
 usedDynamicInputs [&FloatFromCurve]
 usedModules [
  &self:SystemSpawnScript.SystemScriptSource.SystemScriptGraph.NiagaraNodeAssignment_2.SetVariables_#1
  &SystemState
  &self:SystemSpawnScript.SystemScriptSource.SystemScriptGraph.NiagaraNodeAssignment_1.SetVariables_#2
 ]
}
emitters {
 Trail_Grey {
  emitterUpdateScript [
   &EmitterState(Life Cycle Mode=::NewEnumerator1, Loop Behavior=::NewEnumerator1, Loop Duration=0.5, UseLoopDelay=false)
   &FN_SpawnAnimTrail()
  ]
  particleSpawnScript [
   &FN_InitializeAnimTrailBoneIndecies(First Bone or Socket Index=1, Second Bone or Socket Index=6)
   &FN_InitializeAnimTrail(Color=<rgba 0.2571 0.2571 0.3 1>, Lifetime=0.2)
  ]
  particleUpdateScript [
   &ParticleState() {enabled=false}
   &FN_UpdateAnimTrail()
   &ScaleColor(
    Scale Alpha=&Multiply_Float(A=$System.SystemFade, B=1)
    Scale RGB=&MakeVectorFromLinearColorRGB(Color=&FNTimeOfDayColor(Night Color=<rgba 0.4792 0.6719 1 1>, Morning Color=<rgba 0.6 0.6262 0.8 1>, Day Color=<rgba 1 1 1 1>, Evening Color=<rgba 0.6 0.6 0.6 1>))
    ScaleA=true
    ScaleRGB=true
   )
  ]
  rendererClasses [&NiagaraRibbonRendererProperties]
  renderers [{rendererClass=&NiagaraRibbonRendererProperties}]
 }
 Trail_Red {
  emitterUpdateScript [
   &EmitterState(Life Cycle Mode=::NewEnumerator1, Loop Behavior=::NewEnumerator1, Loop Duration=0.5, UseLoopDelay=false)
   &FN_SpawnAnimTrail()
  ]
  particleSpawnScript [
   &FN_InitializeAnimTrailBoneIndecies(First Bone or Socket Index=3, Second Bone or Socket Index=7)
   &FN_InitializeAnimTrail(AnimTrailWidth=1.5, Color=<rgba 0.3438 0.0507 0.0501 1>, Lifetime=0.2)
  ]
  particleUpdateScript [
   &ParticleState() {enabled=false}
   &FN_UpdateAnimTrail()
   &ScaleColor(
    Scale Alpha=&Multiply_Float(A=$System.SystemFade, B=1)
    Scale RGB=&MakeVectorFromLinearColorRGB(Color=&FNTimeOfDayColor(Night Color=<rgba 0.4792 0.6719 1 1>, Morning Color=<rgba 0.6 0.6262 0.8 1>, Day Color=<rgba 1 1 1 1>, Evening Color=<rgba 0.6 0.6 0.6 1>))
    ScaleA=true
    ScaleRGB=true
   )
  ]
  rendererClasses [&NiagaraRibbonRendererProperties]
  renderers [{rendererClass=&NiagaraRibbonRendererProperties}]
 }
 Trail_Yellow {
  emitterUpdateScript [
   &EmitterState(Life Cycle Mode=::NewEnumerator1, Loop Behavior=::NewEnumerator1, Loop Duration=0.5, UseLoopDelay=false)
   &FN_SpawnAnimTrail()
  ]
  particleSpawnScript [
   &FN_InitializeAnimTrailBoneIndecies(First Bone or Socket Index=1, Second Bone or Socket Index=5)
   &FN_InitializeAnimTrail(Color=<rgba 0.4115 0.2623 0.1185 1>, Lifetime=0.2)
  ]
  particleUpdateScript [
   &ParticleState() {enabled=false}
   &FN_UpdateAnimTrail()
   &ScaleColor(
    Scale Alpha=&Multiply_Float(A=$System.SystemFade, B=1)
    Scale RGB=&MakeVectorFromLinearColorRGB(Color=&FNTimeOfDayColor(Night Color=<rgba 0.4792 0.6719 1 1>, Morning Color=<rgba 0.6 0.6262 0.8 1>, Day Color=<rgba 1 1 1 1>, Evening Color=<rgba 0.6 0.6 0.6 1>))
    ScaleA=true
    ScaleRGB=true
   )
  ]
  rendererClasses [&NiagaraRibbonRendererProperties]
  renderers [{rendererClass=&NiagaraRibbonRendererProperties}]
 }
}
summary {
 systemName=NS_Pickaxe_WrenchPeel_Trail_Offhand
 emitters [
  {emitterName=Trail_Yellow, rendererClasses [&NiagaraRibbonRendererProperties]}
  {emitterName=Trail_Red, rendererClasses [&NiagaraRibbonRendererProperties]}
  {emitterName=Trail_Grey, rendererClasses [&NiagaraRibbonRendererProperties]}
 ]
 userVariables [User.RibbonIDOffset {type=NiagaraInt32}]
}
